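# Animation Blueprint for the Raptor backpack cosmetic that adds secondary physics motion to its dangling bones.
# The local reference pose runs through chained Anim Dynamics nodes for the bag cover, strap and flask, with angular limits, heavy damping and overridden gravity.
# Each update, bIsTargeting drops the strap and flask alpha from full to 0.2, damping their sway while targeting.
# Left and right towel dynamics nodes exist but are not connected to the output pose.
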AnimBlueprint /BRCosmetics/Accessories/FORT_Backpacks/F_MED_Raptor_Backpack/Meshes/F_MED_Raptor_Pack_AnimBlueprint
target_skeleton=&/F_MED_Raptor_Pack_Skeleton
graphs [
 AnimGraph {nodes [
   n1 = Anim Dynamics - Bone: dyn_l_towel(Alpha=1, ExternalForce="0,0,0") {
    angularDampingOverride=0.85
    bGravityOverrideInSimSpace=true
    bOverrideAngularDamping=true
    bUseGravityOverride=true
    alphaBoolBlend {blendOption=Linear}
    boundBone {boneName=dyn_l_towel}
    componentAppliedLinearAccClamp=<3000 3000 3000>
    componentLinearAccScale=<0.5 0.5 0.5>
    componentLinearVelScale=<0.5 0.5 0.5>
    gravityOverride {z=-980}
    physicsBodyDefinitions [{
      sphereCollisionRadius=0
      boundBone {boneName=dyn_l_towel}
      boxExtents=<5 5 5>
      constraintSetup {angularLimitsMax=<0 0 35>, angularLimitsMin=<0 -35 -35>, angularTarget=<0 0 0>}
      localJointOffset=<0.2 0 0>
     }]
   }
   Anim Dynamics - Bone: dyn_r_towel(ComponentPose=@n1, Alpha=1, ExternalForce="0,0,0") {
    angularDampingOverride=0.85
    bGravityOverrideInSimSpace=true
    bOverrideAngularDamping=true
    bUseGravityOverride=true
    alphaBoolBlend {blendOption=Linear}
    boundBone {boneName=dyn_r_towel}
    componentAppliedLinearAccClamp=<3000 3000 3000>
    componentLinearAccScale=<0.5 0.5 0.5>
    componentLinearVelScale=<0.5 0.5 0.5>
    gravityOverride {z=-980}
    physicsBodyDefinitions [{
      sphereCollisionRadius=0
      boundBone {boneName=dyn_r_towel}
      boxExtents=<5 5 5>
      constraintSetup {angularLimitsMax=<0 0 35>, angularLimitsMin=<0 -35 -35>, angularTarget=<0 0 0>}
      localJointOffset=<0.2 0 0>
     }]
   }
   n2 = Anim Dynamics - Bone: dyn_bag_cover(ComponentPose=Local To Component(LocalPose=Local Space Ref Pose()), Alpha=1, ExternalForce="0,0,0") {
    angularDampingOverride=0.95
    bGravityOverrideInSimSpace=true
    bOverrideAngularDamping=true
    bUseGravityOverride=true
    alphaBoolBlend {blendOption=Linear}
    boundBone {boneName=dyn_bag_cover}
    componentLinearAccScale=<0.5 0.5 0.5>
    componentLinearVelScale=<0.5 0.5 0.5>
    gravityOverride {z=-980}
    physicsBodyDefinitions [{
      sphereCollisionRadius=0
      boundBone {boneName=dyn_bag_cover}
      boxExtents=<5 4 6>
      constraintSetup {angularLimitsMax=<0 0 10>, angularLimitsMin=<0 0 -30>, angularTarget=<0 0 0>}
      localJointOffset=<0.5 0 0>
     }]
   }
   n3 = Get alpha_value() -> (alpha_value:double)
   n4 = Anim Dynamics - Bone: dyn_strap(ComponentPose=@n2, Alpha=@n3, ExternalForce="0,0,0") {
    angularDampingOverride=0.98
    bGravityOverrideInSimSpace=true
    bOverrideAngularDamping=true
    bUseGravityOverride=true
    alphaBoolBlend {blendOption=Linear}
    boundBone {boneName=dyn_strap}
    componentAppliedLinearAccClamp=<1000 1000 1000>
    componentLinearAccScale=<0.6 0.6 0.05>
    componentLinearVelScale=<0.8 0.8 0.1>
    gravityOverride {z=-980}
    physicsBodyDefinitions [{
      sphereCollisionRadius=0
      boundBone {boneName=dyn_strap}
      boxExtents=<8 6 6>
      constraintSetup {angularLimitsMax=<0 30 30>, angularLimitsMin=<0 -5 0>, angularTarget=<0 0 0>}
      localJointOffset=<0.5 0 0>
     }]
   }
   n5 = Anim Dynamics - Bone: dyn_flask(ComponentPose=@n4, Alpha=@n3, ExternalForce="0,0,0") {
    angularDampingOverride=0.98
    bGravityOverrideInSimSpace=true
    bOverrideAngularDamping=true
    bUseGravityOverride=true
    alphaBoolBlend {blendOption=Linear}
    boundBone {boneName=dyn_flask}
    componentAppliedLinearAccClamp=<1000 1000 1000>
    componentLinearAccScale=<0.6 0.6 0.05>
    componentLinearVelScale=<0.8 0.8 0.1>
    gravityOverride {z=-980}
    physicsBodyDefinitions [{
      sphereCollisionRadius=0
      boundBone {boneName=dyn_flask}
      boxExtents=<8 6 8>
      constraintSetup {angularLimitsMax=<0 30 30>, angularLimitsMin=<0 -5 0>, angularTarget=<0 0 0>}
      localJointOffset=<1 0.5 0>
     }]
   }
   Output Pose(Result=Component To Local(ComponentPose=@n5)) {layerGroup=None}
  ]}
 EventGraph {nodes [
   Event BlueprintUpdateAnimation()
   > n1 = Branch(Condition=Get bIsTargeting())
   > Set alpha_value(alpha_value:double=0.2) -> (Output_Get:double)
   Set alpha_value(execute=@n1.else, alpha_value=1)
  ]}
]
variables [alpha_value {type=number}]
